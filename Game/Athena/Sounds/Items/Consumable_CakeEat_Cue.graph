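# Sound cue for eating the cake consumable, routed to the consumables sound class with consumable-use attenuation.
# A mixer layers two borrowed Durr Burger item waves, burger and fries, playing them together with the fries layer slightly boosted.
# A concurrency override caps simultaneous instances and stops the oldest when the limit is exceeded.
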SoundCue /Game/Athena/Sounds/Items/Consumable_CakeEat_Cue
root=Mixer_0
nodes [
 {id=Mixer_0, children [WavePlayer_0,WavePlayer_1], properties {inputVolume [1,1.2]}}
 {id=WavePlayer_0, properties {soundWaveAssetPtr=&/Item_Durrable_Burger_01}}
 {id=WavePlayer_1, properties {soundWaveAssetPtr=&/Item_Durrable_Fries_01}}
]
properties {
 attenuationSettings=&Fort_Consumable_Use_Athena_Att
 bOverrideConcurrency=true
 maxDistance=2250
 soundClassObject=&Fort_Consumables
 volumeMultiplier=0.7
 concurrencyOverrides {maxCount=4, resolutionRule=StopOldest}
}
