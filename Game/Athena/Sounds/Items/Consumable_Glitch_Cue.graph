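# Sound cue for the Glitch consumable in Athena, playing a randomly chosen "swap" sound through the gadgets and items sound class.
# A modulator feeds a quality-level switch, where the first two levels pick evenly from five wave variants and the last two from only two.
# Dedicated attenuation sets its spatial falloff, and a concurrency override caps simultaneous instances, stopping the oldest.
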SoundCue /Game/Athena/Sounds/Items/Consumable_Glitch_Cue
root=Modulator_0
nodes [
 {id=Modulator_0, children [QualityLevel_0]}
 {id=QualityLevel_0, children [Random_0,Random_0,Random_1,Random_1]}
 {id=Random_0, children [WavePlayer_0,WavePlayer_2,WavePlayer_3,WavePlayer_4,WavePlayer_1], properties {weights [1,1,1,1,1]}}
 {id=Random_1, children [WavePlayer_6,WavePlayer_5], properties {weights [1,1]}}
 {id=WavePlayer_0, properties {soundWaveAssetPtr=&/Consumable_Glitch_Swap_05}}
 {id=WavePlayer_1, properties {soundWaveAssetPtr=&/Consumable_Glitch_Swap_01}}
 {id=WavePlayer_2, properties {soundWaveAssetPtr=&/Consumable_Glitch_Swap_04}}
 {id=WavePlayer_3, properties {soundWaveAssetPtr=&/Consumable_Glitch_Swap_03}}
 {id=WavePlayer_4, properties {soundWaveAssetPtr=&/Consumable_Glitch_Swap_02}}
 {id=WavePlayer_5, properties {soundWaveAssetPtr=&Consumable_Glitch_Swap_01}}
 {id=WavePlayer_6, properties {soundWaveAssetPtr=&Consumable_Glitch_Swap_05}}
]
properties {
 attenuationSettings=/Game/Sounds/Attenuation/Fort_Comsumable_Glitch_Att
 bOverrideConcurrency=true
 maxDistance=2250
 soundClassObject=&Fort_Gadgets_Items
 volumeMultiplier=0.4
 concurrencyOverrides {maxCount=4, resolutionRule=StopOldest}
}
