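# Sound cue for the ice cream consumable in Athena, playing one of three ice cream sound variations.
# A Random node picks among the three wave players with equal weight.
# It uses the consumable-use attenuation and the consumables sound class.
# A concurrency override caps simultaneous instances at a small count and stops the oldest when exceeded.
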SoundCue /Game/Athena/Sounds/Items/Consumable_IceCream_Cue
root=Random_0
nodes [
 {id=Random_0, children [WavePlayer_2,WavePlayer_3,WavePlayer_4], properties {weights [1,1,1]}}
 {id=WavePlayer_2, properties {soundWaveAssetPtr=&/Consumable_IceCream_01}}
 {id=WavePlayer_3, properties {soundWaveAssetPtr=&/Consumable_IceCream_02}}
 {id=WavePlayer_4, properties {soundWaveAssetPtr=&/Consumable_IceCream_03}}
]
properties {
 attenuationSettings=&Fort_Consumable_Use_Athena_Att
 bOverrideConcurrency=true
 maxDistance=2250
 soundClassObject=&Fort_Consumables
 volumeMultiplier=1
 concurrencyOverrides {maxCount=4, resolutionRule=StopOldest}
}
