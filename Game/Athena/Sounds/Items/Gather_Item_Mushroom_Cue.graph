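# Sound cue that plays when a mushroom item is gathered in the Athena mode.
# A single wave player outputs the Gather_Item_01 sound directly, with no randomization or mixing.
# It uses the consumable-use attenuation and the consumables sound class.
# A concurrency override allows only one instance at a time, stopping the oldest when a new one starts.
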SoundCue /Game/Athena/Sounds/Items/Gather_Item_Mushroom_Cue
root=WavePlayer_0
nodes [{id=WavePlayer_0, properties {soundWaveAssetPtr=&/Gather_Item_01}}]
properties {
 attenuationSettings=&Fort_Consumable_Use_Athena_Att
 bOverrideConcurrency=true
 maxDistance=2250
 soundClassObject=&Fort_Consumables
 volumeMultiplier=1
 concurrencyOverrides {maxCount=1, resolutionRule=StopOldest}
}
