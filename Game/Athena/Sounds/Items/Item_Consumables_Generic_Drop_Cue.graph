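# Sound cue for dropping a generic consumable item.
# A random node picks between two equally weighted drop waves and feeds a modulator at the root, left at default settings, for per-play variation.
# It plays in the loot collect sound class at reduced volume and references the shared weapon drop attenuation asset.
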SoundCue /Game/Athena/Sounds/Items/Item_Consumables_Generic_Drop_Cue
root=Modulator_0
nodes [
 {id=Modulator_0, children [Random_0]}
 {id=Random_0, children [WavePlayer_1,WavePlayer_2], properties {weights [1,1]}}
 {id=WavePlayer_1, properties {soundWaveAssetPtr=&/Item_Consumables_Generic_Drop_02}}
 {id=WavePlayer_2, properties {soundWaveAssetPtr=&/Item_Consumables_Generic_Drop_01}}
]
properties {
 attenuationSettings=&Fort_Weapon_Drop_Att
 maxDistance=3000
 soundClassObject=&Fort_Loot_Collect
 volumeMultiplier=0.6
 attenuationOverrides {falloffDistance=2400, stereoSpread=2000, attenuationShapeExtents {x=800}}
}
