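# Sound cue for dropping an ice cream consumable item.
# A random node picks between two drop wave variations with equal weight.
# It uses the loot-collect sound class, the shared weapon-drop attenuation asset and a slightly reduced volume.
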SoundCue /Game/Athena/Sounds/Items/Item_Consumables_Icecream_Drop_Cue
root=Random_0
nodes [
 {id=Random_0, children [WavePlayer_0,WavePlayer_3], properties {weights [1,1]}}
 {id=WavePlayer_0, properties {soundWaveAssetPtr=/Game/Athena/Sounds/Items/IceCream/Item_Consumables_IceCream_Drop_02}}
 {id=WavePlayer_3, properties {soundWaveAssetPtr=/Game/Athena/Sounds/Items/IceCream/Item_Consumables_IceCream_Drop_01}}
]
properties {
 attenuationSettings=&Fort_Weapon_Drop_Att
 maxDistance=3000
 soundClassObject=&Fort_Loot_Collect
 volumeMultiplier=0.85
 attenuationOverrides {falloffDistance=2400, stereoSpread=2000, attenuationShapeExtents {x=600}}
}
