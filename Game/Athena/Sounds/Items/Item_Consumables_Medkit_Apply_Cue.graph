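# Sound cue for applying a medkit consumable in Athena (Battle Royale).
# A single wave player plays the MedKit use sound through the Athena consumable-use attenuation and the consumables sound class, at slightly reduced volume.
# A concurrency override caps simultaneous instances and stops the oldest when the limit is exceeded.
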SoundCue /Game/Athena/Sounds/Items/Item_Consumables_Medkit_Apply_Cue
root=WavePlayer_2
nodes [{id=WavePlayer_2, properties {soundWaveAssetPtr=&/Item_Consumables_MedKit_Use}}]
properties {
 attenuationSettings=&Fort_Consumable_Use_Athena_Att
 bOverrideConcurrency=true
 maxDistance=2250
 soundClassObject=&Fort_Consumables
 volumeMultiplier=0.9
 concurrencyOverrides {maxCount=4, resolutionRule=StopOldest}
}
