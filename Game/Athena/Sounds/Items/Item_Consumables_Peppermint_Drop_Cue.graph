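# Sound cue played when a Peppermint consumable item is dropped in Athena, using impact sounds.
# A Random node picks with equal weight between two peppermint impact waves.
# The cue uses the weapon-drop attenuation and the loot-collect sound class, with slightly reduced volume.
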SoundCue /Game/Athena/Sounds/Items/Item_Consumables_Peppermint_Drop_Cue
root=Random_1
nodes [
 {id=Random_1, children [WavePlayer_0,WavePlayer_3], properties {weights [1,1]}}
 {id=WavePlayer_0, properties {soundWaveAssetPtr=/Game/Athena/Sounds/Items/Peppermint/Peppermint_Impact_01}}
 {id=WavePlayer_3, properties {soundWaveAssetPtr=/Game/Athena/Sounds/Items/Peppermint/Peppermint_Impact_02}}
]
properties {
 attenuationSettings=&Fort_Weapon_Drop_Att
 maxDistance=3000
 soundClassObject=&Fort_Loot_Collect
 volumeMultiplier=0.85
 attenuationOverrides {falloffDistance=2400, stereoSpread=2000, attenuationShapeExtents {x=600}}
}
