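# Sound cue for the jostle sound of the shield potion jar consumable, judging by its name during the item's "cook" phase.
# A Random node picks one of four equally weighted large-jar jostle waves and feeds a Modulator that varies pitch for variation.
# The cue plays pitched up and at reduced volume in the gadgets sound class, reusing the weapon reload attenuation.
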SoundCue /Game/Athena/Sounds/Items/Item_Consumables_Shield_Potion_Jar_Cook_Cue
root=Modulator_1
nodes [
 {id=Modulator_1, children [Random_1], properties {pitchMax=1.1}}
 {id=Random_1, children [WavePlayer_4,WavePlayer_5,WavePlayer_6,WavePlayer_7], properties {weights [1,1,1,1]}}
 {id=WavePlayer_4, properties {soundWaveAssetPtr=&/Shield_Jar_Large_Jostle_01}}
 {id=WavePlayer_5, properties {soundWaveAssetPtr=&/Shield_Jar_Large_Jostle_02}}
 {id=WavePlayer_6, properties {soundWaveAssetPtr=&/Shield_Jar_Large_Jostle_03}}
 {id=WavePlayer_7, properties {soundWaveAssetPtr=&/Shield_Jar_Large_Jostle_04}}
]
properties {
 attenuationSettings=&Fort_Weapon_Reload_Att
 maxDistance=2250
 pitchMultiplier=1.5
 soundClassObject=&Fort_Gadgets
 volumeMultiplier=0.4
}
