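# Sound cue that plays when drinking a shield potion jar completes in the Athena game mode.
# A single completion wave runs through a Delay node left at default settings.
# It uses the consumable-use attenuation and consumables sound class, with a volume boost.
# Overridden concurrency caps simultaneous instances, stopping the oldest first.
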SoundCue /Game/Athena/Sounds/Items/Item_Consumables_Shield_Potion_Jar_Drink_Complete_Cue
root=Delay_0
nodes [
 {id=Delay_0, children [WavePlayer_0]}
 {id=WavePlayer_0, properties {soundWaveAssetPtr=&/sfx_ShieldPotion_Complete_01}}
]
properties {
 attenuationSettings=&Fort_Consumable_Use_Athena_Att
 bOverrideConcurrency=true
 maxDistance=2250
 soundClassObject=&Fort_Consumables
 volumeMultiplier=1.25
 concurrencyOverrides {maxCount=4, resolutionRule=StopOldest}
}
