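# Sound cue for drinking the Shield Potion Jar consumable item in Athena.
# A single wave player plays the jar use sound wave directly, with no randomization or modulation nodes.
# It uses the consumables sound class and the shared Athena consumable-use attenuation.
# A concurrency override caps simultaneous instances at four, stopping the oldest when exceeded.
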SoundCue /Game/Athena/Sounds/Items/Item_Consumables_Shield_Potion_Jar_Drink_Cue
root=WavePlayer_3
nodes [{id=WavePlayer_3, properties {soundWaveAssetPtr=&/Item_Consumables_Shield_Potion_Jar_Use}}]
properties {
 attenuationSettings=&Fort_Consumable_Use_Athena_Att
 bOverrideConcurrency=true
 maxDistance=2250
 soundClassObject=&Fort_Consumables
 volumeMultiplier=1
 concurrencyOverrides {maxCount=4, resolutionRule=StopOldest}
}
