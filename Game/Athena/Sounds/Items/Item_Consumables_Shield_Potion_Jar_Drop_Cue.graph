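# Sound cue for the drop sound of the shield potion jar consumable item.
# A single wave player feeds the output with the jar drop wave, with no randomization or modulation.
# It plays at reduced volume in the loot-collect sound class, spatialized through the shared weapon-drop attenuation settings.
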SoundCue /Game/Athena/Sounds/Items/Item_Consumables_Shield_Potion_Jar_Drop_Cue
root=WavePlayer_0
nodes [{id=WavePlayer_0, properties {soundWaveAssetPtr=&/Item_Consumables_Shield_Potion_Jar_Drop}}]
properties {
 attenuationSettings=&Fort_Weapon_Drop_Att
 maxDistance=3000
 soundClassObject=&Fort_Loot_Collect
 volumeMultiplier=0.4
}
