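# Sound cue for equipping the shield potion jar consumable, played as a gadget sound with weapon-reload attenuation.
# A mixer layers two branches, with the hand-grab layer quieter than the jostle layer.
# Each branch runs through a modulator with fixed volume into an equally weighted random pick of four large-jar wave variations, one set for grabs and one for jostles.
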SoundCue /Game/Athena/Sounds/Items/Item_Consumables_Shield_Potion_Jar_Equip_Cue
root=Mixer_0
nodes [
 {id=Mixer_0, children [Modulator_0,Modulator_1], properties {inputVolume [0.7,1]}}
 {id=Modulator_0, children [Random_0], properties {volumeMax=1, volumeMin=1}}
 {id=Modulator_1, children [Random_1], properties {volumeMax=1, volumeMin=1}}
 {
  id=Random_0
  children [WavePlayer_0,WavePlayer_1,WavePlayer_2,WavePlayer_3]
  properties {bShouldExcludeFromBranchCulling=true, weights [1,1,1,1]}
 }
 {
  id=Random_1
  children [WavePlayer_4,WavePlayer_5,WavePlayer_6,WavePlayer_7]
  properties {bShouldExcludeFromBranchCulling=true, weights [1,1,1,1]}
 }
 {id=WavePlayer_0, properties {soundWaveAssetPtr=&/Shield_Jar_Large_Hand_Grab_01}}
 {id=WavePlayer_1, properties {soundWaveAssetPtr=&/Shield_Jar_Large_Hand_Grab_02}}
 {id=WavePlayer_2, properties {soundWaveAssetPtr=&/Shield_Jar_Large_Hand_Grab_03}}
 {id=WavePlayer_3, properties {soundWaveAssetPtr=&/Shield_Jar_Large_Hand_Grab_04}}
 {id=WavePlayer_4, properties {soundWaveAssetPtr=&/Shield_Jar_Large_Jostle_01}}
 {id=WavePlayer_5, properties {soundWaveAssetPtr=&/Shield_Jar_Large_Jostle_02}}
 {id=WavePlayer_6, properties {soundWaveAssetPtr=&/Shield_Jar_Large_Jostle_03}}
 {id=WavePlayer_7, properties {soundWaveAssetPtr=&/Shield_Jar_Large_Jostle_04}}
]
properties {
 attenuationSettings=&Fort_Weapon_Reload_Att
 maxDistance=2250
 soundClassObject=&Fort_Gadgets
 volumeMultiplier=0.7
}
