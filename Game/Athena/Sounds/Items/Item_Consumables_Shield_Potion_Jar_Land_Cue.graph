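# Sound cue for a shield potion jar item landing, played in the loot collect sound class with weapon-drop attenuation.
# A Random node picks evenly among three landing waves reused from the mini shield potion and passes the result through a Modulator at default settings.
# The cue lowers pitch and volume, so the shared mini potion recordings play deeper and quieter for the jar.
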SoundCue /Game/Athena/Sounds/Items/Item_Consumables_Shield_Potion_Jar_Land_Cue
root=Modulator_0
nodes [
 {id=Modulator_0, children [Random_0]}
 {id=Random_0, children [WavePlayer_1,WavePlayer_2,WavePlayer_3], properties {bShouldExcludeFromBranchCulling=true, weights [1,1,1]}}
 {id=WavePlayer_1, properties {soundWaveAssetPtr=&/Item_Consumables_Shield_Potion_Mini_Land_01}}
 {id=WavePlayer_2, properties {soundWaveAssetPtr=&/Item_Consumables_Shield_Potion_Mini_Land_02}}
 {id=WavePlayer_3, properties {soundWaveAssetPtr=&/Item_Consumables_Shield_Potion_Mini_Land_03}}
]
properties {
 attenuationSettings=&Fort_Weapon_Drop_Att
 maxDistance=3000
 pitchMultiplier=0.8
 soundClassObject=&Fort_Loot_Collect
 volumeMultiplier=0.7
}
