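# Sound cue for picking up a shield potion jar in Athena, played through the loot-collect sound class with weapon pickup attenuation.
# A Random node picks with equal weight between two mini shield potion pickup waves, then feeds a Modulator that holds volume fixed.
# The cue reuses the mini potion recordings at lowered pitch and slightly raised volume to distinguish the jar.
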SoundCue /Game/Athena/Sounds/Items/Item_Consumables_Shield_Potion_Jar_Pickup_Cue
root=Modulator_0
nodes [
 {id=Modulator_0, children [Random_0], properties {volumeMax=1, volumeMin=1}}
 {id=Random_0, children [WavePlayer_1,WavePlayer_2], properties {bShouldExcludeFromBranchCulling=true, weights [1,1]}}
 {id=WavePlayer_1, properties {soundWaveAssetPtr=&/Item_Consumables_Shield_Potion_Mini_Pickup_01}}
 {id=WavePlayer_2, properties {soundWaveAssetPtr=&/Item_Consumables_Shield_Potion_Mini_Pickup_02}}
]
properties {
 attenuationSettings=&Fort_Weapon_Pickup_Att
 maxDistance=3500
 pitchMultiplier=0.8
 soundClassObject=&Fort_Loot_Collect
 volumeMultiplier=1.1
}
