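# Sound cue for the whoosh of throwing a shield potion jar consumable in Athena.
# A Random node picks one of three equally weighted large-jar whoosh waves and feeds a Modulator at the root.
# It plays in the gadgets sound class and borrows the weapon projectile in-air loop attenuation asset.
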SoundCue /Game/Athena/Sounds/Items/Item_Consumables_Shield_Potion_Jar_Throw_Cue
root=Modulator_0
nodes [
 {id=Modulator_0, children [Random_2]}
 {id=Random_2, children [WavePlayer_8, WavePlayer_9, WavePlayer_10], properties {weights [1,1,1]}}
 {id=WavePlayer_10, properties {soundWaveAssetPtr=&/Shield_Jar_Large_Throw_Whoosh_03}}
 {id=WavePlayer_8, properties {soundWaveAssetPtr=&/Shield_Jar_Large_Throw_Whoosh_01}}
 {id=WavePlayer_9, properties {soundWaveAssetPtr=&/Shield_Jar_Large_Throw_Whoosh_02}}
]
properties {
 attenuationSettings=/Game/Sounds/Attenuation/Fort_Weapon_Projectile_InAir_Loop_Att
 maxDistance=4500
 soundClassObject=&Fort_Gadgets
}
