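# Sound cue for a mini shield potion consumable, playing a small jar hand-grab sound during its "cook" phase.
# A Random node picks evenly among three grab wave variants, and a Modulator varies pitch on each play.
# The cue runs at a raised base pitch and lowered volume, using weapon-reload attenuation and the gadgets sound class.
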SoundCue /Game/Athena/Sounds/Items/Item_Consumables_Shield_Potion_Mini_Cook_Cue
root=Modulator_0
nodes [
 {id=Modulator_0, children [Random_1], properties {pitchMax=1.1, pitchMin=0.8}}
 {id=Random_1, children [WavePlayer_4,WavePlayer_5,WavePlayer_6], properties {weights [1,1,1]}}
 {id=WavePlayer_4, properties {soundWaveAssetPtr=&/Shield_Jar_Small_Hand_Grab_01}}
 {id=WavePlayer_5, properties {soundWaveAssetPtr=&/Shield_Jar_Small_Hand_Grab_02}}
 {id=WavePlayer_6, properties {soundWaveAssetPtr=&/Shield_Jar_Small_Hand_Grab_03}}
]
properties {
 attenuationSettings=&Fort_Weapon_Reload_Att
 maxDistance=2250
 pitchMultiplier=1.5
 soundClassObject=&Fort_Gadgets
 volumeMultiplier=0.4
}
