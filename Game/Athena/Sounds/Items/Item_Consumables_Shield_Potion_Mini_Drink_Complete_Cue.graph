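# Sound cue for completing a mini shield potion drink in Athena.
# A single wave player feeds the root directly, with no randomization or mixing, and the cue raises both pitch and volume.
# It uses the shared consumable-use attenuation and consumables sound class, with a concurrency override that stops the oldest instance when the limit is exceeded.
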SoundCue /Game/Athena/Sounds/Items/Item_Consumables_Shield_Potion_Mini_Drink_Complete_Cue
root=WavePlayer_0
nodes [{id=WavePlayer_0, properties {soundWaveAssetPtr=&/sfx_MiniShieldPotion_Complete_01}}]
properties {
 attenuationSettings=&Fort_Consumable_Use_Athena_Att
 bOverrideConcurrency=true
 maxDistance=2250
 pitchMultiplier=1.25
 soundClassObject=&Fort_Consumables
 volumeMultiplier=4
 concurrencyOverrides {maxCount=4, resolutionRule=StopOldest}
}
