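# Sound cue for dropping a mini shield potion consumable in Athena.
# A single wave player for the drop sound feeds a modulator whose volume range is pinned flat, so it adds no volume variation.
# The cue plays quietly through a reduced volume multiplier, using weapon-drop attenuation and the loot-collect sound class.
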SoundCue /Game/Athena/Sounds/Items/Item_Consumables_Shield_Potion_Mini_Drop_Cue
root=Modulator_0
nodes [
 {id=Modulator_0, children [WavePlayer_0], properties {volumeMax=1, volumeMin=1}}
 {id=WavePlayer_0, properties {soundWaveAssetPtr=&/Item_Consumables_Shield_Potion_Mini_Drop}}
]
properties {
 attenuationSettings=&Fort_Weapon_Drop_Att
 maxDistance=3000
 soundClassObject=&Fort_Loot_Collect
 volumeMultiplier=0.3
}
