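# Sound cue for throwing a mini shield potion, playing a whoosh in the Athena item audio set.
# A single wave player with the small shield jar throw whoosh feeds a modulator that randomizes pitch around normal on each play.
# It plays in the gadgets sound class and borrows an in-air weapon projectile loop attenuation for spatial falloff.
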SoundCue /Game/Athena/Sounds/Items/Item_Consumables_Shield_Potion_Mini_Throw_Cue
root=Modulator_0
nodes [
 {id=Modulator_0, children [WavePlayer_0], properties {pitchMax=1.2, pitchMin=0.8}}
 {id=WavePlayer_0, properties {soundWaveAssetPtr=&/Shield_Jar_Small_Throw_Whoosh_01}}
]
properties {
 attenuationSettings=/Game/Sounds/Attenuation/Fort_Weapon_Projectile_InAir_Loop_Att
 maxDistance=4500
 soundClassObject=&Fort_Gadgets
}
